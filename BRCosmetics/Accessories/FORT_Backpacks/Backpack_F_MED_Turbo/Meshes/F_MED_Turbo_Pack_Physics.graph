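# Physics asset for the Turbo backpack cosmetic mesh, providing collision bodies and joints for the pack and its four wheels.
# One long capsule covers the pack bone, and a smaller capsule sits on each front and back, left and right dyn_wheel bone.
# Each wheel is constrained directly to the pack with free twist and narrow swing limits, allowing it to spin while staying closely aligned.
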PhysicsAsset /BRCosmetics/Accessories/FORT_Backpacks/Backpack_F_MED_Turbo/Meshes/F_MED_Turbo_Pack_Physics
bodies {
 dyn_wheelLB [{shapeType=Capsule, length=9.3567, radius=1.956, center=<0.1481 0.1021 2.4892>, rotation=<rot 2.3679 -89.9996 4.4129>}]
 dyn_wheelLF [{shapeType=Capsule, length=7.7543, radius=1.6758, center=<-0.021 -0.0989 1.0283>, rotation=<rot -5.0017 -89.9999 3.9099>}]
 dyn_wheelRB [{shapeType=Capsule, length=9.3293, radius=1.9827, center=<0.09 0.0731 -1.7311>, rotation=<rot -2.4583 -90 -4.4355>}]
 dyn_wheelRF [{shapeType=Capsule, length=7.8255, radius=1.6265, center=<-0.187 -0.1109 1.5152>, rotation=<rot 5.0591 -89.9999 -3.8239>}]
 pack [{shapeType=Capsule, length=40.4533, radius=4.3602, center=<-0.0028 -4.7963 -3.5956>, rotation=<rot 9.8864 -90 -0.0122>}]
}
constraints [
 {bone1Name=dyn_wheelRF, bone2Name=pack, swing1LimitDegrees=10, swing2LimitDegrees=10, twistLimitDegrees=45, twistMotion=Free}
 {bone1Name=dyn_wheelRB, bone2Name=pack, swing1LimitDegrees=10, swing2LimitDegrees=10, twistLimitDegrees=45, twistMotion=Free}
 {bone1Name=dyn_wheelLB, bone2Name=pack, swing1LimitDegrees=10, swing2LimitDegrees=10, twistLimitDegrees=45, twistMotion=Free}
 {bone1Name=dyn_wheelLF, bone2Name=pack, swing1LimitDegrees=10, swing2LimitDegrees=10, twistLimitDegrees=45, twistMotion=Free}
]
